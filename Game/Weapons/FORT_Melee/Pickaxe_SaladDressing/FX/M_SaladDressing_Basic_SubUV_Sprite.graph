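# Unlit translucent Niagara sprite material for SaladDressing pickaxe FX, defaulting to a 3x2 SubUV impact flipbook.
# The flipbook's blue channel blends two tint colors under particle color; a masked channel drives opacity, with depth and optional view-angle fade skipped on mobile.
# A soft noise-offset radial or linear dissolve follows the larger of manual or dynamic-parameter values.
# Dissolve frame, scale and pan parameters use only X for both axes.
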Material /Game/Weapons/FORT_Melee/Pickaxe_SaladDressing/FX/M_SaladDressing_Basic_SubUV_Sprite
graph {
 properties {bUsedWithNiagaraSprites=true, blendMode=BLEND_Translucent, pixelDepthOffsetMode=PDOM_Legacy, shadingModel=MSM_Unlit, usedShadingModels=Unlit}
 nodes [
  n1 = VectorParameter() {group=03 - Color Transi, parameterName=MainColor, defaultValue=<rgba 1 1 1 0>}
  n2 = VectorParameter() {group=03 - Color Transi, parameterName=SecondColor, defaultValue=<rgba 1 1 1 0>}
  n3 = TextureSampleParameterSubUV!(UVs=ParticleSubUVProperties().TextureCoordinate0, Apply View MipBias) {group=00 - FlipBook, parameterName=FlipBook, texture=&/T_SaladDressing_Impact_3x2}
  n4 = SmoothStep(
   Min=ScalarParameter() {group=03 - Color Transi, parameterName=ColorTransiMin}
   Max=ScalarParameter() {defaultValue=1, group=03 - Color Transi, parameterName=ColorTransiMax}
   Value=Power(Base=@n3.B, Exp=ScalarParameter() {defaultValue=1, group=03 - Color Transi, parameterName=ColorTransiPow})
  )
  n5 = ParticleColor()
  MaterialFunctionCall(Emissive=Multiply(A=LinearInterpolate(A=@n1.RGB, B=@n2.RGB, Alpha=@n4), B=@n5.RGB)) {materialFunction=&MF_FX_EmissiveAdjustments}
  n6 = ScalarParameter() {parameterName=FlipBookSmoothMin}
  n7 = ScalarParameter() {defaultValue=1, parameterName=FlipBookSmoothMax}
  n8 = ChannelMaskParameter(None=@n3.RGB) {group=00 - FlipBook, parameterName=TextureMask}
  n9 = DepthFade(Opacity=@n5.A, FadeDistance=ScalarParameter() {defaultValue=50, group=01 - Opacity, parameterName=DepthFade})
  n10 = ScalarParameter() {group=01 - Opacity, parameterName=Angle Fade Bias}
  n11 = LinearInterpolate(
   A=@n9
   B=Multiply(
    A=@n9
    B=MaterialFunctionCall("FadeSensitivity (optional)"=@n10) {materialFunction=&FadeBasedOnViewAngle}.Result
   )
   Alpha=ScalarParameter() {group=01 - Opacity, parameterName=Use Angle Fade}
  )
  n12 = ScalarParameter() {defaultValue=0.5, group=02 - Dissolve, parameterName=DissolveSoftness, sortPriority=-1}
  n13 = MaterialFunctionCall(
   Input=Max(
    A=ScalarParameter() {group=02 - Dissolve, parameterName=ManualDissolve, sliderMax=1, sortPriority=-2}
    B=DynamicParameter() {defaultValue=<rgba 0 1 1 1>, paramNames [Dissolve,Param2,Param3,Param4]}.Dissolve
   )
   Target Low=Subtract(B=@n12) {constA=0}
  ) {materialFunction=&RemapValueRange}
  n14 = Add(A=@n13.Result, B=@n12)
  n15 = VectorParameter() {group=02 - Dissolve, parameterName=FlipBookFrames, sortPriority=0, channelNames {a="", b="", g=Y, r=X}, defaultValue=<rgba 3 2 0 0>}
  n16 = Frac(None=Multiply(
    A=TextureCoordinate()
    B=MaterialFunctionCall(X=@n15.X, Y=@n15.X) {materialFunction=&MakeFloat2}.Result
   ))
  n17 = Constant2Vector() {g=-0.5, r=-0.5}
  n18 = LinearInterpolate(
   A=@n16
   B=MaterialFunctionCall(Vector or UVs=Add(A=@n16, B=@n17)) {materialFunction=&VectorToRadialValue}.Radial Coordinates
   Alpha=ScalarParameter() {defaultValue=1, group=02 - Dissolve, parameterName=RadialDissolve}
  )
  n19 = VectorParameter() {
   group=02 - Dissolve
   parameterName=DissolveNoiseUV
   sortPriority=1
   channelNames {a=PanY, b=PanX, g=ScaleY, r=ScaleX}
   defaultValue=<rgba 1 1 0 0>
  }
  n20 = MaterialFunctionCall(X=@n19.ScaleX, Y=@n19.ScaleX) {materialFunction=&MakeFloat2}
  n21 = MaterialFunctionCall(X=@n19.PanX, Y=@n19.PanX) {materialFunction=&MakeFloat2}
  n22 = Time()
  n23 = TextureSampleParameter2D(UVs=Add(A=Multiply(A=@n18, B=@n20.Result), B=Multiply(A=@n21.Result, B=@n22))) {
   group=02 - Dissolve
   parameterName=DissolveTexture
   samplerType=SAMPLERTYPE_Grayscale
   texture=/Game/Effects/Fort_Effects/Textures/Fire/T_WispyFireNoise_Soft_M
  }
  n24 = Add(A=@n18, B=Multiply(A=@n23.R, B=ScalarParameter() {group=02 - Dissolve, parameterName=NoiseIntensity}))
  n25 = LinearInterpolate(
   A=ComponentMask(None=@n24) {r=true}
   B=ComponentMask(None=@n24) {g=true}
   Alpha=ScalarParameter() {defaultValue=1, group=02 - Dissolve, parameterName=DissolveOrientation, sortPriority=2}
  )
  n26 = LinearInterpolate(A=@n25, B=OneMinus(None=@n25), Alpha=ScalarParameter() {group=02 - Dissolve, parameterName=InvertDissolve, sortPriority=3})
  n27 = ScalarParameter() {defaultValue=1, group=02 - Dissolve, parameterName=UseGradialDissolve, sortPriority=4}
  n28 = Saturate(None=Saturate(None=LinearInterpolate(A=@n23.R, B=@n26, Alpha=@n27) {constA=1}))
  Multiply(
   A=Multiply(A=SmoothStep(Min=@n6, Max=@n7, Value=@n8.R), B=ShadingPathSwitch(Default=@n11, Mobile=@n5.A))
   B=Saturate(None=SmoothStep(Min=@n13.Result, Max=@n14, Value=@n28))
  )
 ]
}
